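# Animation blueprint for the player's Mustache gadget skeleton, holding the mesh in its weapon pose.
# A two-state machine switches between Default and ReducedWork on bWantsReducedWork, which is not declared in this blueprint.
# Default passes the pose through DefaultSlot so montages can play, while ReducedWork plays the same pose without the slot.
# The EventGraph holds only unconnected stub nodes, with no update logic.
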AnimBlueprint /Game/Animation/Game/MainPlayer/Gadgets/Mustache/Mustache_SkBP
target_skeleton=/Game/Gadgets/Assets/Mustache/Meshes/Mustache_Gadget_Mustache_Gadget_LOD0_Skeleton
graphs [
 AnimGraph {nodes [
   AnimGraphNode_StateMachine_0 = New State Machine() {is=title AnimGraphNode_StateMachine}
   Output Pose(Result=@AnimGraphNode_StateMachine_0) {layerGroup=None}
  ]}
 New State Machine {
  entry=Default
  path=AnimGraph.AnimGraphNode_StateMachine_0.New State Machine
  ids=id
  nodes [
   AnimStateNode_0 = Default() {is=title AnimStateNode}
   AnimStateNode_1 = ReducedWork() {is=title AnimStateNode, stateFullyBlended {guid=#1, notifyName=FullyBlendedReducedWork}}
   AnimStateTransitionNode_0 = AnimStateTransitionNode()
   AnimStateTransitionNode_1 = AnimStateTransitionNode()
  ]
  links [
   {
    dst=ReducedWork
    rule=AnimGraph.AnimGraphNode_StateMachine_0.New State Machine.AnimStateTransitionNode_0.Transition
    src=Default
    via=AnimStateTransitionNode_0
   }
   {
    dst=Default
    rule=AnimGraph.AnimGraphNode_StateMachine_0.New State Machine.AnimStateTransitionNode_1.Transition
    src=ReducedWork
    via=AnimStateTransitionNode_1
   }
  ]
 }
 Default {
  path=AnimGraph.AnimGraphNode_StateMachine_0.New State Machine.AnimStateNode_0.Default
  nodes [Output Animation Pose(Result=Slot 'DefaultSlot'(Source=Sequence Player 'Mustache_Weapon_Pose_Weapon'() {sequence=&/Mustache_Weapon_Pose_Weapon}))]
 }
 ReducedWork {
  path=AnimGraph.AnimGraphNode_StateMachine_0.New State Machine.AnimStateNode_1.ReducedWork
  nodes [Output Animation Pose(Result=Sequence Player 'Mustache_Weapon_Pose_Weapon'() {sequence=&Mustache_Weapon_Pose_Weapon})]
 }
 Transition {
  path=AnimGraph.AnimGraphNode_StateMachine_0.New State Machine.AnimStateTransitionNode_0.Transition
  nodes [Result(bCanEnterTransition=Get bWantsReducedWork() -> (bWantsReducedWork:bool))]
 }
 Transition {
  path=AnimGraph.AnimGraphNode_StateMachine_0.New State Machine.AnimStateTransitionNode_1.Transition
  nodes [Result(bCanEnterTransition=NOT Boolean(A=Get bWantsReducedWork()))]
 }
 EventGraph {nodes [
   TryGetPawnOwner()
   Event BlueprintUpdateAnimation()
  ]}
]
variables []
